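# Masked, unlit material for the logo shell on the Creative Mannequin device, reusing the Character Spawner's logo shell texture.
# A channel mask parameter, "Alpha Channel", selects a channel from that texture, presumably for the opacity cutout.
# A "Color" vector parameter, black by default, supplies an adjustable flat color.
# The graph lists no explicit links from these nodes to material outputs.
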Material /Game/Creative/Devices/Mannequin/Materials/M_CP_Device_Mannequin_01_LogoShell
graph {
 properties {blendMode=BLEND_Masked, pixelDepthOffsetMode=PDOM_Legacy, shadingModel=MSM_Unlit, usedShadingModels=Unlit}
 nodes [
  ChannelMaskParameter(None=TextureSample() {texture=/Game/Creative/Devices/CharacterSpawner/Textures/T_CP_Device_CharacterSpawner_LogoShell_01}.RGBA) {parameterName=Alpha Channel}
  VectorParameter() {parameterName=Color, defaultValue=<rgba 0 0 0 1>}
 ]
}
